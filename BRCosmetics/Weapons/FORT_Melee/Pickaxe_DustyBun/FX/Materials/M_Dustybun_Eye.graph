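# Translucent unlit Niagara sprite material that draws an animated cartoon eye for the DustyBun pickaxe effects.
# Dynamic parameters drive opening and offset, though EyeXOffset feeds both axes and EyeYOffset is unused.
# A scaled sphere mask forms the iris, with sine-based rings or a solid option, over a shaded sclera.
# Smoothstep thresholds on a grayscale eye texture form eyelid and opacity, tinted by particle color and alpha.
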Material /BRCosmetics/Weapons/FORT_Melee/Pickaxe_DustyBun/FX/Materials/M_Dustybun_Eye
graph {
 properties {
  bAutomaticallySetUsageInEditor=false
  bUsedWithNiagaraSprites=true
  blendMode=BLEND_Translucent
  previewMesh=/Engine/Transient.MaterialEditorMeshComponent_3
  shadingModel=MSM_Unlit
  usedShadingModels=Unlit
 }
 nodes [
  n1 = VectorParameter() {group=Color, parameterName=ScleraShadowColor, defaultValue=<rgba 0 0 0 1>}
  n2 = VectorParameter() {group=Color, parameterName=ScleraColor, defaultValue=<rgba 0.4219 0 0.0026 1>}
  n3 = TextureSampleParameter2D() {
   parameterName=EyeTexture
   samplerType=SAMPLERTYPE_LinearGrayscale
   texture=/BRCosmetics/Weapons/FORT_Melee/Pickaxe_DustyBun/FX/Textures/T_Dustybun_Eye_A
  }
  n4 = DynamicParameter() {defaultValue=<rgba 0 0 0 1>, paramNames [EyeOpening,EyeXOffset,EyeYOffset,Param4]}
  n5 = AppendVector(A=@n4.EyeXOffset, B=@n4.EyeXOffset)
  n6 = AppendVector(
   A=ScalarParameter() {defaultValue=0.8, group=Eye, parameterName=IrisScaleX}
   B=ScalarParameter() {defaultValue=1, group=Eye, parameterName=IrisScaleY}
  )
  n7 = LinearInterpolate(A=@n5, B=Multiply(A=@n5, B=@n6), Alpha=ScalarParameter() {group=Eye, parameterName=MultiplyOffsetByScale})
  n8 = TextureCoordinate()
  n9 = DynamicParameter() {defaultValue=<rgba 0 0 0 1>, paramNames [EyeOpening,EyeXOffset,EyeYOffset,Param4]}
  n10 = Power(Base=@n9.EyeOpening) {constExponent=1.5}
  n11 = SphereMask(
   A=MaterialFunctionCall(UVs=Add(A=@n7, B=@n8) {constA=-0.1}, Texture Scale=@n6) {materialFunction=&ScaleUVsByCenter}.UVs
   B=Constant2Vector() {g=0.5, r=0.5}
   Radius=LinearInterpolate(Alpha=@n10) {constA=0.3, constB=0}
  ) {hardnessPercent=0}
  n12 = VectorParameter() {group=Color, parameterName=IrisOutlineColor, defaultValue=<rgba 0 0 0 1>}
  n13 = ScalarParameter() {group=Eye, parameterName=RingHardnessMin}
  n14 = ScalarParameter() {defaultValue=1, group=Eye, parameterName=RingHardnessMax}
  n15 = ScalarParameter() {defaultValue=4, group=Eye, parameterName=RingCountMax}
  n16 = ScalarParameter() {defaultValue=1, group=Eye, parameterName=RingCountMin}
  n17 = Sine(None=Multiply(A=@n11, B=LinearInterpolate(A=@n15, B=@n16, Alpha=@n10) {constA=4}))
  n18 = LinearInterpolate(Alpha=@n9.EyeOpening) {constA=0.1}
  n19 = SmoothStep(Min=@n18, Max=Add(A=@n18) {constB=0.2}, Value=@n3.R)
  n20 = Saturate(None=Subtract(A=InverseLinearInterpolate(A=@n13, B=@n14, Value=@n17), B=Step(Y=@n19) {constX=0.6, constY=0.9}))
  n21 = ScalarParameter() {group=Eye, parameterName=Solid Iris}
  n22 = Step(Y=@n20) {constX=0, constY=1}
  n23 = LinearInterpolate(
   A=LinearInterpolate(
    A=LinearInterpolate(A=@n1.RGB, B=@n2.RGB, Alpha=@n3.R)
    B=VectorParameter() {group=Color, parameterName=IrisColor, defaultValue=<rgba 1 0 0.0061 1>}.RGB
    Alpha=@n11
   )
   B=LinearInterpolate(A=@n12.RGB, B=Multiply(A=@n12.RGB, B=@n20), Alpha=@n21)
   Alpha=LinearInterpolate(A=@n20, B=OneMinus(None=@n22), Alpha=@n21)
  )
  n24 = Add(A=@n18, B=ScalarParameter() {defaultValue=0.05, desc=EyelidThickness, group=Eye, parameterName=EyelidThickness}) {constB=0.05}
  n25 = Subtract(A=@n19, B=SmoothStep(Min=@n24, Max=Add(A=@n24) {constB=0.175}, Value=@n3.R))
  n26 = Multiply(A=@n25, B=VectorParameter() {group=Color, parameterName=EyelidColor, defaultValue=<rgba 0.025 0.025 0.025 1>}.RGB)
  n27 = SmoothStep(Value=@n25) {constMax=0.2, constMin=0.1}
  n28 = ParticleColor()
  MaterialFunctionCall(Emissive=Multiply(A=LinearInterpolate(A=@n23, B=@n26, Alpha=@n27), B=@n28.RGB)) {materialFunction=&MF_FX_EmissiveAdjustments}
  Multiply(A=Saturate(None=SmoothStep(Value=@n19) {constMax=0.2, constMin=0.1}), B=@n28.A)
 ]
}
