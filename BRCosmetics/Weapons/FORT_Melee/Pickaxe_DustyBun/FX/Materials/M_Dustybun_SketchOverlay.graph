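# Translucent unlit two-sided Niagara sprite material that draws an animated pencil-sketch overlay for the DustyBun pickaxe effects.
# Time quantized to 12 steps per second jumps the panned pencil-streak texture and cycles its RGB channel blend, optionally with per-particle randomness, for a hand-drawn flipbook look.
# The inverted streaks are subtracted from a radial mask, then multiplied by particle color for emissive and particle alpha for opacity.
# Opacity supports an optional hard step and depth fade.
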Material /BRCosmetics/Weapons/FORT_Melee/Pickaxe_DustyBun/FX/Materials/M_Dustybun_SketchOverlay
graph {
 properties {
  bAutomaticallySetUsageInEditor=false
  bUsedWithNiagaraSprites=true
  blendMode=BLEND_Translucent
  shadingModel=MSM_Unlit
  twoSided=true
  usedShadingModels=Unlit
 }
 nodes [
  n1 = ScalarParameter() {defaultValue=0.2, parameterName=Sharpness}
  n2 = TextureCoordinate()
  n3 = Constant() {r=0.5}
  n4 = Constant()
  n5 = ScalarParameter() {defaultValue=1, parameterName=Tiling_U}
  n6 = ScalarParameter() {defaultValue=1, parameterName=Tiling_V}
  n7 = Ceil(None=Multiply(A=Time(), B=Constant() {r=12}) {constB=3})
  n8 = Panner(Coordinate=Multiply(A=TextureCoordinate() {uTiling=1.5, vTiling=2}, B=AppendVector(A=@n5, B=@n6)), Time=Multiply(A=Constant() {r=2}, B=@n7)) {speedX=0.1, speedY=-0.1}
  n9 = Constant3Vector() {constant=<rgba 1 0 0 0>}
  n10 = Constant3Vector() {constant=<rgba 0 1 0 0>}
  n11 = Constant3Vector() {constant=<rgba 0 0 1 0>}
  n12 = Multiply(A=@n7) {constA=0.1, constB=0.2}
  n13 = Multiply(A=@n12, B=ParticleRandom())
  n14 = MaterialFunctionCall(Value=StaticSwitchParameter(True=Add(A=@n13) {constB=-1}, False=@n12) {parameterName=ParticleRandom}) {materialFunction=&LinearSine}
  n15 = OneMinus(None=DotProduct(
    A=TextureSample(UVs=@n8) {samplerType=SAMPLERTYPE_Masks, texture=/BRCosmetics/Weapons/FORT_Melee/Pickaxe_DustyBun/FX/Textures/T_PencilStreaks}.RGB
    B=MaterialFunctionCall(A=@n9, B=@n10, C=@n11, Alpha=@n14.Linear Sine) {materialFunction=&Lerp_3Color}.Result
   ))
  n16 = Saturate(None=Subtract(A=SphereMask(A=@n2, B=@n3, Radius=@n3, Hardness=@n4), B=@n15))
  n17 = ParticleColor()
  n18 = Multiply(A=@n16, B=@n17.A)
  n19 = Multiply(
   A=StaticSwitchParameter(True=Step(Y=@n1, X=@n18) {constY=0.2}, False=@n18) {parameterName=Step}
   B=ScalarParameter() {defaultValue=1, parameterName=Intensity}
  ) {constB=5}
  LinearInterpolate(
   A=@n19
   B=DepthFade(Opacity=@n19, FadeDistance=ScalarParameter() {parameterName=FadeDistance})
   Alpha=ScalarParameter() {parameterName=UseDepthFade}
  )
  MaterialFunctionCall(Emissive=Multiply(A=@n16, B=@n17.RGB)) {materialFunction=&MF_FX_EmissiveAdjustments}
  Multiply(A=@n15) {constB=1.5}
 ]
}
