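# Translucent unlit Niagara mesh and ribbon material for a pencil-sketch trail on the DustyBun pickaxe.
# Stepped time pans pencil streaks and cycles their RGB channels for a hand-drawn flicker.
# The streaks erode a soft sphere mask, shaped by edge fades, particle color and depth fade outside forward and mobile paths.
# The Use SplineThicken switch, off by default, enables offsets combining spline thickening, sine wobble and a clamped dynamic-parameter pull.
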Material /BRCosmetics/Weapons/FORT_Melee/Pickaxe_DustyBun/FX/Materials/M_Dustybun_SketchTrail
graph {
 properties {
  bAutomaticallySetUsageInEditor=false
  bUsedWithNiagaraMeshParticles=true
  bUsedWithNiagaraRibbons=true
  blendMode=BLEND_Translucent
  shadingModel=MSM_Unlit
  twoSided=true
  usedShadingModels=Unlit
 }
 nodes [
  n1 = SphereMask(A=TextureCoordinate() {uTiling=0}, B=Constant2Vector() {g=0.5}, Radius=Constant() {r=0.5}, Hardness=Constant() {r=0.5})
  n2 = TextureCoordinate()
  n3 = AppendVector(A=ScalarParameter() {defaultValue=1, parameterName=SketchTileX}, B=ScalarParameter() {defaultValue=1, parameterName=SketchTileY})
  n4 = Ceil(None=Multiply(A=Time(), B=Constant() {r=12}) {constB=3})
  n5 = Panner(
   Coordinate=LinearInterpolate(A=TextureCoordinate() {coordinateIndex=1}, B=Multiply(A=@n2, B=@n3), Alpha=ScalarParameter() {parameterName=TextureSetSwitch})
   Time=Multiply(A=Constant() {r=2}, B=@n4)
  ) {speedX=0.1, speedY=-0.1}
  n6 = Constant3Vector() {constant=<rgba 1 0 0 0>}
  n7 = Constant3Vector() {constant=<rgba 0 1 0 0>}
  n8 = Constant3Vector() {constant=<rgba 0 0 1 0>}
  n9 = MaterialFunctionCall(Value=Multiply(A=@n4) {constA=0.1, constB=0.2}) {materialFunction=&LinearSine}
  n10 = OneMinus(None=DotProduct(
    A=TextureSample(UVs=@n5) {samplerType=SAMPLERTYPE_Masks, texture=/BRCosmetics/Weapons/FORT_Melee/Pickaxe_DustyBun/FX/Textures/T_PencilStreaks}.RGB
    B=MaterialFunctionCall(A=@n6, B=@n7, C=@n8, Alpha=@n9.Linear Sine) {materialFunction=&Lerp_3Color}.Result
   ))
  n11 = ComponentMask(None=TextureCoordinate()) {r=true}
  n12 = Add(A=@n11, B=ScalarParameter() {defaultValue=-0.05, group="Fade Leading && Trailing Edges", parameterName=Y_Leading Bias, sortPriority=2})
  n13 = ScalarParameter() {defaultValue=3, group="Fade Leading && Trailing Edges", parameterName=Y_Leading Scale}
  n14 = LinearInterpolate(
   B=Saturate(None=Multiply(A=@n12, B=@n13) {constB=3})
   Alpha=ScalarParameter() {defaultValue=1, group="Fade Leading && Trailing Edges", parameterName=Y_Fade Leading Edge, sortPriority=4}
  ) {constA=1}
  n15 = Add(
   A=OneMinus(None=@n11)
   B=ScalarParameter() {defaultValue=-0.05, group="Fade Leading && Trailing Edges", parameterName=Y_Trailing Bias, sortPriority=3}
  )
  n16 = ScalarParameter() {defaultValue=2, group="Fade Leading && Trailing Edges", parameterName=Y_Trailing Scale, sortPriority=1}
  n17 = LinearInterpolate(
   B=Saturate(None=Multiply(A=@n15, B=@n16) {constB=2})
   Alpha=ScalarParameter() {defaultValue=1, group="Fade Leading && Trailing Edges", parameterName=Y_Fade Trailing Edge, sortPriority=5}
  ) {constA=1}
  n18 = ComponentMask(None=TextureCoordinate()) {g=true}
  n19 = Add(A=@n18, B=ScalarParameter() {defaultValue=0.1, group="Fade Leading && Trailing Edges", parameterName=X_Leading Bias, sortPriority=2})
  n20 = ScalarParameter() {defaultValue=3, group="Fade Leading && Trailing Edges", parameterName=X_Leading Scale}
  n21 = LinearInterpolate(
   B=Saturate(None=Multiply(A=@n19, B=@n20) {constB=3})
   Alpha=ScalarParameter() {defaultValue=1, group="Fade Leading && Trailing Edges", parameterName=X_Fade Leading Edge, sortPriority=4}
  ) {constA=1}
  n22 = Add(
   A=OneMinus(None=@n18)
   B=ScalarParameter() {defaultValue=0.1, group="Fade Leading && Trailing Edges", parameterName=X_Trailing Bias, sortPriority=3}
  )
  n23 = ScalarParameter() {defaultValue=3, group="Fade Leading && Trailing Edges", parameterName=X_Trailing Scale, sortPriority=1}
  n24 = LinearInterpolate(
   B=Saturate(None=Multiply(A=@n22, B=@n23) {constB=2})
   Alpha=ScalarParameter() {defaultValue=1, group="Fade Leading && Trailing Edges", parameterName=X_Fade Trailing Edge, sortPriority=5}
  ) {constA=1}
  n25 = Multiply(A=Saturate(None=Subtract(A=@n1, B=@n10)), B=Multiply(A=Multiply(A=@n14, B=@n17), B=Multiply(A=@n21, B=@n24)))
  n26 = DepthFade(Opacity=@n25, FadeDistance=ScalarParameter() {defaultValue=5, parameterName=DepthFade})
  n27 = ShadingPathSwitch(Default=@n26, Deferred=@n26, Forward=@n25, Mobile=@n25)
  n28 = ParticleColor()
  MaterialFunctionCall(Emissive=Multiply(A=@n27, B=@n28.RGB)) {materialFunction=&MF_FX_EmissiveAdjustments}
  Multiply(A=@n27, B=@n28.A)
  ScalarParameter() {parameterName=Use Default UV WPO}
  ScalarParameter() {parameterName=Use Default UV WPO}
  ScalarParameter() {parameterName=Use Default UV WPO}
  n29 = DynamicParameter() {defaultValue=<rgba 0 10 0 0>, paramNames [WidthBase, WidthTip, WPO Offset, Use Default UV WPO]}
  n30 = Multiply(
   A=ScalarParameter() {defaultValue=1, parameterName=SplineWidth}
   B=SphereMask(A=TextureCoordinate() {vTiling=0}, B=Constant2Vector() {r=0.5}) {attenuationRadius=0.5, hardnessPercent=50}
  )
  n31 = MaterialFunctionCall(WidthBase=Multiply(A=@n29.WidthBase, B=@n30), WidthTip=Multiply(A=@n29.WidthTip, B=@n30), Expand U or V UV Channel=StaticBool()) {materialFunction=&SplineThicken}
  n32 = LinearInterpolate(
   A=TextureCoordinate() {coordinateIndex=1}
   B=TextureCoordinate()
   Alpha=DynamicParameter() {defaultValue=<rgba 0 10 0 0>, paramNames [WidthBase, WidthTip, WPO Offset, Use Default UV WPO]}.Use Default UV WPO
  )
  n33 = ScalarParameter() {defaultValue=-0.3, parameterName=WPO_Speed}
  n34 = Add(
   A=Add(
    A=ComponentMask(None=@n32) {r=true}
    B=DynamicParameter() {defaultValue=<rgba 0 10 0 0>, paramNames [WidthBase, WidthTip, WPO Offset, Use Default UV WPO]}.WPO Offset
   )
   B=Multiply(A=Time(), B=@n33) {constB=-0.3}
  )
  n35 = LinearInterpolate(
   A=TextureCoordinate() {coordinateIndex=1}
   B=TextureCoordinate()
   Alpha=DynamicParameter() {defaultValue=<rgba 0 10 0 0>, paramNames [WidthBase, WidthTip, WPO Offset, Use Default UV WPO]}.Use Default UV WPO
  )
  n36 = Add(
   A=Add(
    A=ComponentMask(None=@n35) {r=true}
    B=DynamicParameter() {defaultValue=<rgba 0 10 0 0>, paramNames [WidthBase, WidthTip, WPO Offset, Use Default UV WPO]}.WPO Offset
   )
   B=Multiply(A=Time(), B=@n33) {constB=-0.3}
  )
  n37 = Add(
   A=Multiply(A=Constant3Vector() {constant=<rgba 1 0 0 0>}, B=Sine(None=@n34) {period=0.5})
   B=Multiply(A=Constant3Vector() {constant=<rgba 0 1 0 0>}, B=Sine(None=@n36) {period=0.8})
  )
  n38 = ScalarParameter() {defaultValue=0.5, parameterName=WPO Intensity}
  n39 = LinearInterpolate(
   A=TextureCoordinate() {coordinateIndex=1}
   B=TextureCoordinate()
   Alpha=DynamicParameter() {defaultValue=<rgba 0 10 0 0>, paramNames [WidthBase, WidthTip, WPO Offset, Use Default UV WPO]}.Use Default UV WPO
  )
  n40 = Multiply(A=Power(Base=ComponentMask(None=@n39) {r=true})) {constB=3}
  n41 = Transform(None=Multiply(A=Multiply(A=@n37, B=@n38), B=Saturate(None=@n40))) {transformSourceType=TRANSFORMSOURCE_Local}
  n42 = Constant()
  n43 = ComponentMask(None=TextureCoordinate()) {r=true}
  n44 = ScalarParameter() {defaultValue=2, parameterName=WPOBlendPower}
  n45 = DynamicParameter() {parameterIndex=1, defaultValue=<rgba 0 10 0 1>, paramNames [X,Y,Z,WPOIntensity]}
  n46 = MaterialFunctionCall(X=@n45.X, Y=@n45.X, Z=@n45.X) {materialFunction=&MakeFloat3}
  n47 = ParticlePositionWS()
  n48 = LinearInterpolate(
   B=Clamp(
    None=Subtract(A=@n46.Result, B=@n47)
    Min=ScalarParameter() {defaultValue=-30, parameterName=Clamp Min}
    Max=ScalarParameter() {defaultValue=30, parameterName=Clamp Max}
   )
   Alpha=Multiply(A=@n45.WPOIntensity, B=Power(Base=@n43, Exp=@n44))
  )
  n49 = ScalarParameter() {defaultValue=1, parameterName=Use Interpolate WPO}
  StaticSwitchParameter(True=Add(A=Add(A=@n31.WorldPositionOffset, B=@n41), B=LinearInterpolate(A=@n42, B=@n48, Alpha=@n49)), False=Constant()) {parameterName=Use SplineThicken}
 ]
}
