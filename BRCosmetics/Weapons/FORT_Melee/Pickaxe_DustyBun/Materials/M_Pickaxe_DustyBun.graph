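# Master material for the DustyBun pickaxe cosmetic, built on skeletal character material functions and reusing the DustyBun body's material logic.
# Output from the 2023 CharacterParent PreFX function runs through the body's PBR/toon mix and main body functions, then into the matching PostFX function.
# PreFX also feeds mask, pre-skinned position, distance blend and fresnel inputs downstream.
# It uses material attributes, with the shading model chosen by the material expression.
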Material /BRCosmetics/Weapons/FORT_Melee/Pickaxe_DustyBun/Materials/M_Pickaxe_DustyBun
graph {
 properties {
  bUseMaterialAttributes=true
  bUsedWithMeshDeformer=true
  bUsedWithMorphTargets=true
  bUsedWithSkeletalMesh=true
  pixelDepthOffsetMode=PDOM_Legacy
  shadingModel=MSM_FromMaterialExpression
  usedShadingModels=Default Lit | Subsurface | Clear Coat | Subsurface Profile | Cloth
 }
 nodes [
  n1 = MaterialFunctionCall() {materialFunction=&MF_CharacterParent_2023_PreFX}
  n2 = MaterialFunctionCall!(MA_In=@n1.Result, M=@n1.Result) {materialFunction=/BRCosmetics/Characters/Player/Male/Large/Bodies/M_LRG_DustyBun/Materials/MF_DustyBun_Body_PBRToonMix}
  MaterialFunctionCall(
   Material=MaterialFunctionCall!(In_MaterialAttributes=NamedRerouteDeclaration(None=@n2.MA_Out) {name=ToonPBRMix, nodeColor=<rgba 0.125 1 0 1>}) {materialFunction=/BRCosmetics/Characters/Player/Male/Large/Bodies/M_LRG_DustyBun/Materials/MF_DustyBun_Body}.Result
   M=@n1.M
   Pre Skinned Local Position=@n1.M
   Distance Blend=@n1.M
   Fresnel=@n1.M
  ) {materialFunction=&MF_CharacterParent_2023_PostFX}
 ]
}
