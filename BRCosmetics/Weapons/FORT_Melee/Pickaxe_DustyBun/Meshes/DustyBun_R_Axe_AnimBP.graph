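# Animation Blueprint for the right axe of the DustyBun pickaxe cosmetic, layering procedural bone adjustments over an input pose.
# The AnimGraph scales and offsets the root bone, chains additive rotation nodes for index, thumb, pinky and middle finger bones, then offsets the root again.
# Each update computes phase-shifted sine-wave claw rotators from real time; their link to the bone nodes is not shown.
# A customization-loaded event flags large male characters in Large Character.
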AnimBlueprint /BRCosmetics/Weapons/FORT_Melee/Pickaxe_DustyBun/Meshes/DustyBun_R_Axe_AnimBP
target_skeleton=&/DustyBun_R_Axe_Skeleton
graphs [
 AnimGraph {nodes [
   AnimGraphNode_SubInput_0 = Input Pose()
   n1 = "Transform (Modify) Bone - Bone: b_AR_Root"(
    ComponentPose=Local To Component(LocalPose=@AnimGraphNode_SubInput_0)
    Alpha=1
    Translation="-1,-2.2,5.7"
    Rotation="0,0,0"
    Scale="1.2,1.2,1.2"
   ) {
    scaleMode=BMM_Additive
    scaleSpace=BCS_BoneSpace
    translationMode=BMM_Additive
    translationSpace=BCS_BoneSpace
    alphaBoolBlend {blendOption=Linear}
    boneToModify {boneName=b_AR_Root}
    scale=<1.2 1.2 1.2>
    translation=<-1 -2.2 5.7>
   }
   n2 = "Transform (Modify) Bone - Bone: Index_01_R_Axe"(ComponentPose=@n1, Alpha=1, Translation="0,0,0", Rotation="0,0,0", Scale="1,1,1") {rotationMode=BMM_Additive, rotationSpace=BCS_BoneSpace, alphaBoolBlend {blendOption=Linear}, boneToModify {boneName=Index_01_R_Axe}}
   n3 = "Transform (Modify) Bone - Bone: Index_02_R_Axe"(ComponentPose=@n2, Alpha=1, Translation="0,0,0", Rotation="0,0,0", Scale="1,1,1") {rotationMode=BMM_Additive, rotationSpace=BCS_BoneSpace, alphaBoolBlend {blendOption=Linear}, boneToModify {boneName=Index_02_R_Axe}}
   n4 = "Transform (Modify) Bone - Bone: Index_03_R_Axe"(ComponentPose=@n3, Alpha=1, Translation="0,0,0", Rotation="0,0,0", Scale="1,1,1") {rotationMode=BMM_Additive, rotationSpace=BCS_BoneSpace, alphaBoolBlend {blendOption=Linear}, boneToModify {boneName=Index_03_R_Axe}}
   n5 = "Transform (Modify) Bone - Bone: Thumb_01_R_Axe"(ComponentPose=@n4, Alpha=1, Translation="0,0,0", Rotation="0,0,0", Scale="1,1,1") {rotationMode=BMM_Additive, rotationSpace=BCS_BoneSpace, alphaBoolBlend {blendOption=Linear}, boneToModify {boneName=Thumb_01_R_Axe}}
   n6 = "Transform (Modify) Bone - Bone: Thumb_02_R_Axe"(ComponentPose=@n5, Alpha=1, Translation="0,0,0", Rotation="0,0,0", Scale="1,1,1") {rotationMode=BMM_Additive, rotationSpace=BCS_BoneSpace, alphaBoolBlend {blendOption=Linear}, boneToModify {boneName=Thumb_02_R_Axe}}
   n7 = "Transform (Modify) Bone - Bone: Thumb_03_R_Axe"(ComponentPose=@n6, Alpha=1, Translation="0,0,0", Rotation="0,0,0", Scale="1,1,1") {rotationMode=BMM_Additive, rotationSpace=BCS_BoneSpace, alphaBoolBlend {blendOption=Linear}, boneToModify {boneName=Thumb_03_R_Axe}}
   n8 = "Transform (Modify) Bone - Bone: Pinky_01_R_Axe"(ComponentPose=@n7, Alpha=1, Translation="0,0,0", Rotation="0,0,0", Scale="1,1,1") {rotationMode=BMM_Additive, rotationSpace=BCS_BoneSpace, alphaBoolBlend {blendOption=Linear}, boneToModify {boneName=Pinky_01_R_Axe}}
   n9 = "Transform (Modify) Bone - Bone: Pinky_02_R_Axe"(ComponentPose=@n8, Alpha=1, Translation="0,0,0", Rotation="0,0,0", Scale="1,1,1") {rotationMode=BMM_Additive, rotationSpace=BCS_BoneSpace, alphaBoolBlend {blendOption=Linear}, boneToModify {boneName=Pinky_02_R_Axe}}
   n10 = "Transform (Modify) Bone - Bone: Pinky_03_R_Axe"(ComponentPose=@n9, Alpha=1, Translation="0,0,0", Rotation="0,0,0", Scale="1,1,1") {rotationMode=BMM_Additive, rotationSpace=BCS_BoneSpace, alphaBoolBlend {blendOption=Linear}, boneToModify {boneName=Pinky_03_R_Axe}}
   n11 = "Transform (Modify) Bone - Bone: Middle_01_R_Axe"(ComponentPose=@n10, Alpha=1, Translation="0,0,0", Rotation="0,0,0", Scale="1,1,1") {rotationMode=BMM_Additive, rotationSpace=BCS_BoneSpace, alphaBoolBlend {blendOption=Linear}, boneToModify {boneName=Middle_01_R_Axe}}
   n12 = "Transform (Modify) Bone - Bone: Middle_02_R_Axe"(ComponentPose=@n11, Alpha=1, Translation="0,0,0", Rotation="0,0,0", Scale="1,1,1") {rotationMode=BMM_Additive, rotationSpace=BCS_BoneSpace, alphaBoolBlend {blendOption=Linear}, boneToModify {boneName=Middle_02_R_Axe}}
   n13 = "Transform (Modify) Bone - Bone: Middle_03_R_Axe"(ComponentPose=@n12, Alpha=1, Translation="0,0,0", Rotation="0,0,0", Scale="1,1,1") {rotationMode=BMM_Additive, rotationSpace=BCS_BoneSpace, alphaBoolBlend {blendOption=Linear}, boneToModify {boneName=Middle_03_R_Axe}}
   n14 = "Transform (Modify) Bone - Bone: b_AR_Root"(ComponentPose=@n13, Alpha=1, Translation="0,0,-2.5", Rotation="0,0,0", Scale="1,1,1") {translationMode=BMM_Additive, alphaBoolBlend {blendOption=Linear}, boneToModify {boneName=b_AR_Root}, translation {z=-2.5}}
   Output Pose(Result=Component To Local(ComponentPose=@n14)) {layerGroup=None}
  ]}
 EventGraph {nodes [
   Event BlueprintUpdateAnimation()
   n1 = float * float(A=GetRealTimeSeconds(), B=2) {is=title K2Node_PromotableOperator}
   n2 = float * float(A="Sin (Radians)"(A=float + float(A=@n1, B=1.5)), B=4)
   > Set ClawRot4(ClawRot4:Rotator=MakeRotator(Yaw=float + float(A=@n2, B=-5))) -> (Output_Get:Rotator)
   n3 = float * float(A=GetRealTimeSeconds(), B=1.5)
   n4 = float * float(A="Sin (Radians)"(A=float + float(A=@n3, B=0.5)), B=4)
   > Set ClawRot3(ClawRot3:Rotator=MakeRotator(Pitch=float + float(A=@n4, B=-1))) -> (Output_Get:Rotator)
   n5 = float * float(A=GetRealTimeSeconds(), B=1.8)
   n6 = float * float(A="Sin (Radians)"(A=float + float(A=@n5, B=0.7)), B=4)
   > Set ClawRot0(ClawRot0:Rotator=MakeRotator(Pitch=float + float(A=@n6, B=-1))) -> (Output_Get:Rotator)
   n7 = float * float(A=GetRealTimeSeconds(), B=2)
   n8 = float * float(A="Sin (Radians)"(A=float + float(A=@n7, B=2)), B=5)
   > Set ClawRot2(ClawRot2:Rotator=MakeRotator(Yaw=float + float(A=@n8, B=-10))) -> (Output_Get:Rotator)
   n9 = float * float(A=GetRealTimeSeconds(), B=2)
   n10 = float * float(A="Sin (Radians)"(A=float + float(A=@n9, B=2.5)), B=6)
   > Set ClawRot1(ClawRot1:Rotator=MakeRotator(Yaw=float + float(A=@n10, B=-8))) -> (Output_Get:Rotator)
   n11 = float * float(A=GetRealTimeSeconds(), B=2)
   n12 = float * float(A="Sin (Radians)"(A=float + float(A=@n11, B=2.5)), B=10)
   > Set ClawRot01(ClawRot01:Rotator=MakeRotator(Yaw=float + float(A=@n12, B=-12))) -> (Output_Get:Rotator)
   n13 = Event OnCharacterAndPickaxeCustomizationLoaded!() -> (
    OutputDelegate
    then
    CustomBaseMesh:Skeletal Mesh Component Object Reference
    CustomParts:Array of Custom Character Part Object References
    CustomMeshParts:Array of Skeletal Mesh Component Budgeted Object References
   ) {is=title K2Node_Event}
   n14 = Get Array Item with EFort Custom Part Type(Array:Array of Custom Character Part Object References=@n13.CustomParts) -> (Output:Custom Character Part Object Reference)
   n15 = Get BodyTypesPermitted!(self:Custom Character Part Object Reference=@n14) -> (BodyTypesPermitted:EFortCustomBodyType Enum=NONE)
   n16 = Get GenderPermitted!(self:Custom Character Part Object Reference=@n14) -> (GenderPermitted:EFortCustomGender Enum=Invalid)
   > Set Large Character(Large Character:bool=AND Boolean(
     A="Equal (Enum)"(A:EFortCustomBodyType Enum=@n15, B:EFortCustomBodyType Enum=Large)
     B="Equal (Enum)"(A:EFortCustomGender Enum=@n16, B:EFortCustomGender Enum=Male)
    )) -> (Output_Get:bool)
   TryGetPawnOwner()
  ]}
]
variables [
 ClawRot0 {type=Rotator}
 ClawRot01 {type=Rotator}
 ClawRot1 {type=Rotator}
 ClawRot2 {type=Rotator}
 ClawRot3 {type=Rotator}
 ClawRot4 {type=Rotator}
 Large Character {type=boolean}
]
